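; Editor art-tool actor that samples base color, specular, roughness and metallic from scene captures and previews the result on a sphere.
; Construction creates the sphere's dynamic material, binds four capture components to render targets, creating any missing ones, then runs sampling.
; SampleMaterial reads each target at a UV with the editor-only sample node and writes the values to variables and material parameters.
; CreateMaterialSwatch passes those stored values to BP_MaterialSwatches as a new swatch.
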(event EventBeginPlay)

(event Collision|EventActorBeginOverlap
  (params (OtherActor :type Actor%20Object%20Reference)))

(event EventTick
  (params (DeltaSeconds :type Float%20%28single-precision%29)))

(event Custom|SampleRenderTargets
  (CallFunction|SampleMaterial 0.5))

(event Custom|SampleMaterial
  (params (x :type Float%20%28double-precision%29 :default 0.5) (y :type Float%20%28double-precision%29 :default 0.5))
  (bind _returnvalue_r (Math|Color|BreakColor (Rendering|RenderTargetSampleUVEditorOnly (Variables|Default|GetTextureRenderTarget_Metallic) (Variables|Default|GetUV))))
  (bind _uv (Variables|Default|GetUV))
  (bind _texturerendertarget_basecolor (Variables|Default|GetTextureRenderTarget_BaseColor))
  (bind _returnvalue (Rendering|RenderTargetSampleUVEditorOnly _texturerendertarget_basecolor _uv))
  (bind _returnvalue_r_2 (Math|Color|BreakColor (Rendering|RenderTargetSampleUVEditorOnly (Variables|Default|GetTextureRenderTarget_Specular) (Variables|Default|GetUV))))
  (bind _returnvalue_r_3 (Math|Color|BreakColor (Rendering|RenderTargetSampleUVEditorOnly (Variables|Default|GetTextureRenderTarget_Roughness) (Variables|Default|GetUV))))
  (Variables|Default|SetUV (Math|Vector2D|MakeVector2D x y))
  (Utilities|FlowControl|Sequence
    (:then_0
      (Development|PrintString 0 (Utilities|String|ToString(LinearColor) _returnvalue))
      (bind _output_get (Variables|Default|SetBaseColor _returnvalue))
      (Rendering|Material|SetVectorParameterValue (Variables|Default|GetSphere_DynamicMaterialInstance) "BaseColor_MM" _output_get :!class /Script/Engine.MaterialInstanceDynamic))
    (:then_1
      (bind _output_get_2 (Variables|Default|SetS _returnvalue_r_2))
      (Rendering|Material|SetScalarParameterValue (Variables|Default|GetSphere_DynamicMaterialInstance) "Specular_MM" _output_get_2 :!class /Script/Engine.MaterialInstanceDynamic))
    (:then_2
      (bind _output_get_3 (Variables|Default|SetR _returnvalue_r_3))
      (Rendering|Material|SetScalarParameterValue (Variables|Default|GetSphere_DynamicMaterialInstance) "Roughness_MM" _output_get_3 :!class /Script/Engine.MaterialInstanceDynamic))
    (:then_3
      (bind _output_get_4 (Variables|Default|SetM _returnvalue_r))
      (Rendering|Material|SetScalarParameterValue (Variables|Default|GetSphere_DynamicMaterialInstance) "Metallic_MM" _output_get_4 :!class /Script/Engine.MaterialInstanceDynamic))))

(event Custom|CreateMaterialSwatch
  (if (not (Class|ItemUIFunctionLibrary|IsNullItem (Variables|Default|GetBP_MaterialSwatches)))
    (Class|BPMaterialSwatches|AddNewSwatchwithParameters (Variables|Default|GetBP_MaterialSwatches) (Variables|Default|GetBaseColor) (Variables|Default|GetS) (Variables|Default|GetR) (Variables|Default|GetM))))

(fn CreateRenderTargets
  (bind _texturerendertarget_basecolor (Variables|Default|GetTextureRenderTarget_BaseColor))
  (bind _texturerendertarget_specular (Variables|Default|GetTextureRenderTarget_Specular))
  (bind _texturerendertarget_metallic (Variables|Default|GetTextureRenderTarget_Metallic))
  (bind _rendertarget_resolution (Variables|Default|GetrenderTarget_resolution))
  (bind _texturerendertarget_roughness (Variables|Default|GetTextureRenderTarget_Roughness))
  (Utilities|FlowControl|Sequence
    (:then_0
      (Utilities|IsValid _texturerendertarget_basecolor
        (:Is%20Valid
          (Variables|/Script/Engine.SceneCaptureComponent2D|SetTextureTarget _texturerendertarget_basecolor (Variables|Default|GetSceneCaptureComponent2D_baseColor)))
        (:Is%20Not%20Valid
          (bind _returnvalue (Rendering|CreateRenderTarget2D 0 _rendertarget_resolution _rendertarget_resolution "RTF_RGBA16f" "(R=1.000000,G=0.000000,B=0.000000,A=1.000000)"))
          (bind _output_get (Variables|Default|SetTextureRenderTarget_BaseColor _returnvalue))
          (Variables|/Script/Engine.SceneCaptureComponent2D|SetTextureTarget _output_get (Variables|Default|GetSceneCaptureComponent2D_baseColor)))))
    (:then_1
      (Utilities|IsValid _texturerendertarget_roughness
        (:Is%20Valid
          (Variables|/Script/Engine.SceneCaptureComponent2D|SetTextureTarget _texturerendertarget_roughness (Variables|Default|GetSceneCaptureComponent2D_roughness)))
        (:Is%20Not%20Valid
          (bind _returnvalue_2 (Rendering|CreateRenderTarget2D 0 _rendertarget_resolution _rendertarget_resolution "RTF_RGBA16f" "(R=1.000000,G=0.000000,B=0.000000,A=1.000000)"))
          (bind _output_get_2 (Variables|Default|SetTextureRenderTarget_Roughness _returnvalue_2))
          (Variables|/Script/Engine.SceneCaptureComponent2D|SetTextureTarget _output_get_2 (Variables|Default|GetSceneCaptureComponent2D_roughness)))))
    (:then_2
      (Utilities|IsValid _texturerendertarget_specular
        (:Is%20Valid
          (Variables|/Script/Engine.SceneCaptureComponent2D|SetTextureTarget _texturerendertarget_specular (Variables|Default|GetSceneCaptureComponent2D_specular)))
        (:Is%20Not%20Valid
          (bind _returnvalue_3 (Rendering|CreateRenderTarget2D 0 _rendertarget_resolution _rendertarget_resolution "RTF_RGBA16f" "(R=1.000000,G=0.000000,B=0.000000,A=1.000000)"))
          (bind _output_get_3 (Variables|Default|SetTextureRenderTarget_Specular _returnvalue_3))
          (Variables|/Script/Engine.SceneCaptureComponent2D|SetTextureTarget _output_get_3 (Variables|Default|GetSceneCaptureComponent2D_specular)))))
    (:then_3
      (Utilities|IsValid _texturerendertarget_metallic
        (:Is%20Valid
          (Variables|/Script/Engine.SceneCaptureComponent2D|SetTextureTarget _texturerendertarget_metallic (Variables|Default|GetSceneCaptureComponent2D_metallic)))
        (:Is%20Not%20Valid
          (bind _returnvalue_4 (Rendering|CreateRenderTarget2D 0 _rendertarget_resolution _rendertarget_resolution "RTF_RGBA16f" "(R=1.000000,G=0.000000,B=0.000000,A=1.000000)"))
          (bind _output_get_4 (Variables|Default|SetTextureRenderTarget_Metallic _returnvalue_4))
          (Variables|/Script/Engine.SceneCaptureComponent2D|SetTextureTarget _output_get_4 (Variables|Default|GetSceneCaptureComponent2D_metallic)))))))

(fn ConstructionScript
  (bind _sphere (Variables|Default|GetSphere))
  (bind _returnvalue (Rendering|Material|CreateDynamicMaterialInstance _sphere 0 "/Game/ArtTools/MaterialLibrary/Materials/M_FN_Character_MASTER_DSRM_parameters.M_FN_Character_MASTER_DSRM_parameters" "None" :!class /Script/Engine.PrimitiveComponent))
  (bind _output_get (Variables|Default|SetSphere_DynamicMaterialInstance _returnvalue))
  (Rendering|Material|SetMaterial _sphere 0 _output_get)
  (CallFunction|CreateRenderTargets)
  (CallFunction|SampleRenderTargets))
