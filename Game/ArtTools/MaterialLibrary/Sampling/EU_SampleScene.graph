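# Editor utility widget that samples scene buffers under the mouse and creates material swatches from buttons.
# While capture is on (button or Space), Tick sends the mouse position to level actors via `ke * SampleMaterial`; Create Material sends `ke * CreateMaterialSwatch`, requiring BP_MaterialSampler and BP_MaterialsSwatches.
# Both axes are divided by desktop width, with an aspect-derived Y offset from Construct, so correct positions require fullscreen.
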EditorUtilityWidgetBlueprint /Game/ArtTools/MaterialLibrary/Sampling/EU_SampleScene
graphs [
 EventGraph {nodes [
   "On Clicked (Button_Toggle)"() {is=title K2Node_ComponentBoundEvent}
   > Toggle()
   "On Clicked (Button_CreateMaterial)"() {is=title K2Node_ComponentBoundEvent}
   > ExecuteConsoleCommand(Command=ke * CreateMaterialSwatch)
   Toggle() {is=title K2Node_CustomEvent}
   > Set Capture(Capture:bool=NOT Boolean(A=Get Capture() -> (Capture:bool))) -> (Output_Get:bool)
   Event Construct()
   > n1 = GetGameUserSettings!(execute) -> (then, ReturnValue:Game User Settings Object Reference)
   n2 = GetDesktopResolution!(self:Game User Settings Object Reference=@n1) -> (ReturnValue_X:int, ReturnValue_Y:int)
   n3 = float / float(A="To Float (Integer)"(InInt=int - int(A=@n2, B=@n2.ReturnValue_Y) {is=title K2Node_CallFunction}), B=2) {is=title K2Node_CallFunction}
   n4 = "To Float (Integer)"(InInt=@n2)
   > Set YOffset(YOffset:double=float - float(A=float / float(A=@n3, B=@n4), B=-0.01) {is=title K2Node_CallFunction}) -> (Output_Get:double)
   > Set DesktopResolution(DesktopResolution:Vector 2D Structure=MakeVector2D(X=@n4, Y="To Float (Integer)"(InInt=@n2.ReturnValue_Y))) -> (Output_Get:Vector 2D Structure)
   Event Tick()
   > Branch(Condition=Get Capture())
   > n5 = getMousePosition!(execute, self) -> (then, mousePosition_X:double, mousePosition_Y:double)
   > ExecuteConsoleCommand(Command=Append!(A="ke * SampleMaterial ", B="To String (Float)"(InDouble=@n5), C:String=" ", D:String="To String (Float)"(InDouble=@n5.mousePosition_Y)) -> (ReturnValue))
   Event PreConstruct()
  ]}
 OnPreviewKeyDown {nodes [
   n1 = OnPreviewKeyDown!() -> (then, MyGeometry:Geometry Structure, InKeyEvent:Key Event Structure) {is=title K2Node_FunctionEntry}
   > Branch(Condition="Equal (Key)"!(A:Key Structure=GetKey!(Input:"Key Event Structure (by ref)"=@n1.InKeyEvent) -> (ReturnValue:Key Structure), B:Key Structure=SpaceBar) -> (ReturnValue:bool))
   > PrintString(InString=Hello, TextColor=(R=1,G=0,B=0.1049,A=1))
   > Toggle() {is=title K2Node_CallFunction}
   > Return Node!(execute, ReturnValue:Event Reply Structure=Handled())
  ]}
 getMousePosition {nodes [
   getMousePosition() {is=title K2Node_FunctionEntry}
   > n1 = GetMousePositionOnViewport!(execute, WorldContextObject:Object Reference) -> (then, ReturnValue_X:double, ReturnValue_Y:double)
   n2 = Get!() -> (DesktopResolution_X:double, DesktopResolution_Y:double)
   n3 = MakeVector2D(
    X="Clamp (Float)"(Value=float / float(A=@n1, B=@n2), Min=0, Max=1)
    Y=float + float(A=float / float(A=@n1.ReturnValue_Y, B=@n2), B=Get yOffset() -> (yOffset:double)) {is=title K2Node_CommutativeAssociativeBinaryOperator}
   )
   > PrintString(InString="To String (Vector2d)"!(InVec:Vector 2D Structure=@n3) -> (ReturnValue:String))
   > Return Node!(execute, mousePosition:Vector 2D Structure=@n3)
  ]}
]
variables [
 Capture {type=boolean}
 DesktopResolution {type=Vector2D, default=<0 0>}
 YOffset {type=number}
]
widgets {
 info {rootWidgetClass=&CanvasPanel, widgetCount=5}
 widgets [
  CanvasPanel_0 {widgetClassPath=&CanvasPanel, widget=CanvasPanel_0}
  Button_Toggle {
   widgetClassPath=&Button
   parent=CanvasPanel_0
   widget=Button_Toggle
   properties {
    toolTipText="Samples Buffers using the Mouse Position. You must be full screen (F11) for the mouse position to be reported correctly."
    widgetStyle {
     clickedSlateSound {resourceObject=null}
     disabled {
      bIsDynamicallyLoaded=false
      drawAs=RoundedBox
      imageType=NoImage
      mirroring=NoMirror
      resourceName=None
      resourceObject=null
      tiling=NoTile
      imageSize=<0 0>
      margin=<ltrb 0 0 0 0>
      outlineSettings {
       bUseBrushTransparency=false
       roundingType=FixedRadius
       width=1
       color {colorUseRule=UseColor_ColorTable, specifiedColor=<rgba 0 0 0 0>}
       cornerRadii=<4 4 4 4>
      }
      tintColor {colorUseRule=UseColor_Specified, specifiedColor=<rgba 0.0395 0.0395 0.0395 1>}
      uVRegion {bIsValid=false, max=<0 0>, min=<0 0>}
     }
     disabledForeground {colorUseRule=UseColor_ColorTable, specifiedColor=<rgba 0 0 0 0>}
     hovered {
      bIsDynamicallyLoaded=false
      drawAs=RoundedBox
      imageType=NoImage
      mirroring=NoMirror
      resourceName=None
      resourceObject=null
      tiling=NoTile
      imageSize=<0 0>
      margin=<ltrb 0 0 0 0>
      outlineSettings {
       bUseBrushTransparency=false
       roundingType=FixedRadius
       width=1
       color {colorUseRule=UseColor_ColorTable, specifiedColor=<rgba 0 0 0 0>}
       cornerRadii=<4 4 4 4>
      }
      tintColor {colorUseRule=UseColor_Specified, specifiedColor=<rgba 0.0953 0.0953 0.0953 1>}
      uVRegion {bIsValid=false, max=<0 0>, min=<0 0>}
     }
     hoveredForeground {colorUseRule=UseColor_ColorTable, specifiedColor=<rgba 0 0 0 0>}
     hoveredSlateSound {resourceObject=null}
     normal {
      bIsDynamicallyLoaded=false
      drawAs=RoundedBox
      imageType=NoImage
      mirroring=NoMirror
      resourceName=None
      resourceObject=null
      tiling=NoTile
      imageSize=<0 0>
      margin=<ltrb 0 0 0 0>
      outlineSettings {
       bUseBrushTransparency=false
       roundingType=FixedRadius
       width=1
       color {colorUseRule=UseColor_ColorTable, specifiedColor=<rgba 0 0 0 0>}
       cornerRadii=<4 4 4 4>
      }
      tintColor {colorUseRule=UseColor_Specified, specifiedColor=<rgba 0.0395 0.0395 0.0395 1>}
      uVRegion {bIsValid=false, max=<0 0>, min=<0 0>}
     }
     normalForeground {colorUseRule=UseColor_ColorTable, specifiedColor=<rgba 0 0 0 0>}
     normalPadding=<ltrb 12 1 12 1>
     pressed {
      bIsDynamicallyLoaded=false
      drawAs=RoundedBox
      imageType=NoImage
      mirroring=NoMirror
      resourceName=None
      resourceObject=null
      tiling=NoTile
      imageSize=<0 0>
      margin=<ltrb 0 0 0 0>
      outlineSettings {
       bUseBrushTransparency=false
       roundingType=FixedRadius
       width=1
       color {colorUseRule=UseColor_ColorTable, specifiedColor=<rgba 0 0 0 0>}
       cornerRadii=<4 4 4 4>
      }
      tintColor {colorUseRule=UseColor_Specified, specifiedColor=<rgba 0.0284 0.0284 0.0284 1>}
      uVRegion {bIsValid=false, max=<0 0>, min=<0 0>}
     }
     pressedForeground {colorUseRule=UseColor_ColorTable, specifiedColor=<rgba 0 0 0 0>}
     pressedPadding=<ltrb 12 2 12 0>
     pressedSlateSound {resourceObject=null}
    }
   }
   slot {
    class=CanvasPanelSlot
    properties {layoutData {alignment=<0 0>, anchors {maximum=<0 0>, minimum=<0 0>}, offsets=<ltrb 16.902 16.178 409.4003 83.862>}}
   }
  }
  CommonTextBlock_114 {
   widgetClassPath=&CommonTextBlock
   parent=Button_Toggle
   widget=CommonTextBlock_114
   properties {
    justification=Center
    style=&TextStyle-Base_C
    text=Toggle Sampling
    font {
     bForceMonospaced=false
     bMaterialIsStencil=false
     fontMaterial=null
     fontObject=&NotoSans
     letterSpacing=0
     monospacedWidth=1
     size=24
     skewAmount=0
     typefaceFontName=NotoSans-Regular
     outlineSettings {
      bApplyOutlineToDropShadows=false
      bMiteredCorners=false
      bSeparateFillAlpha=false
      outlineMaterial=null
      outlineSize=0
      outlineColor=<rgba 0 0 0 1>
     }
    }
    shadowOffset=<0 0>
   }
   slot {class=ButtonSlot}
  }
  Button_CreateMaterial {
   widgetClassPath=&Button
   parent=CanvasPanel_0
   widget=Button_CreateMaterial
   properties {
    toolTipText=Creates a material. Needs both a BP_MaterialSampler and a BP_MaterialsSwatches to be set up in your level.
    widgetStyle {
     clickedSlateSound {resourceObject=null}
     disabled {
      bIsDynamicallyLoaded=false
      drawAs=RoundedBox
      imageType=NoImage
      mirroring=NoMirror
      resourceName=None
      resourceObject=null
      tiling=NoTile
      imageSize=<0 0>
      margin=<ltrb 0 0 0 0>
      outlineSettings {
       bUseBrushTransparency=false
       roundingType=FixedRadius
       width=1
       color {colorUseRule=UseColor_ColorTable, specifiedColor=<rgba 0 0 0 0>}
       cornerRadii=<4 4 4 4>
      }
      tintColor {colorUseRule=UseColor_Specified, specifiedColor=<rgba 0.0395 0.0395 0.0395 1>}
      uVRegion {bIsValid=false, max=<0 0>, min=<0 0>}
     }
     disabledForeground {colorUseRule=UseColor_ColorTable, specifiedColor=<rgba 0 0 0 0>}
     hovered {
      bIsDynamicallyLoaded=false
      drawAs=RoundedBox
      imageType=NoImage
      mirroring=NoMirror
      resourceName=None
      resourceObject=null
      tiling=NoTile
      imageSize=<0 0>
      margin=<ltrb 0 0 0 0>
      outlineSettings {
       bUseBrushTransparency=false
       roundingType=FixedRadius
       width=1
       color {colorUseRule=UseColor_ColorTable, specifiedColor=<rgba 0 0 0 0>}
       cornerRadii=<4 4 4 4>
      }
      tintColor {colorUseRule=UseColor_Specified, specifiedColor=<rgba 0.0953 0.0953 0.0953 1>}
      uVRegion {bIsValid=false, max=<0 0>, min=<0 0>}
     }
     hoveredForeground {colorUseRule=UseColor_ColorTable, specifiedColor=<rgba 0 0 0 0>}
     hoveredSlateSound {resourceObject=null}
     normal {
      bIsDynamicallyLoaded=false
      drawAs=RoundedBox
      imageType=NoImage
      mirroring=NoMirror
      resourceName=None
      resourceObject=null
      tiling=NoTile
      imageSize=<0 0>
      margin=<ltrb 0 0 0 0>
      outlineSettings {
       bUseBrushTransparency=false
       roundingType=FixedRadius
       width=1
       color {colorUseRule=UseColor_ColorTable, specifiedColor=<rgba 0 0 0 0>}
       cornerRadii=<4 4 4 4>
      }
      tintColor {colorUseRule=UseColor_Specified, specifiedColor=<rgba 0.0395 0.0395 0.0395 1>}
      uVRegion {bIsValid=false, max=<0 0>, min=<0 0>}
     }
     normalForeground {colorUseRule=UseColor_ColorTable, specifiedColor=<rgba 0 0 0 0>}
     normalPadding=<ltrb 12 1 12 1>
     pressed {
      bIsDynamicallyLoaded=false
      drawAs=RoundedBox
      imageType=NoImage
      mirroring=NoMirror
      resourceName=None
      resourceObject=null
      tiling=NoTile
      imageSize=<0 0>
      margin=<ltrb 0 0 0 0>
      outlineSettings {
       bUseBrushTransparency=false
       roundingType=FixedRadius
       width=1
       color {colorUseRule=UseColor_ColorTable, specifiedColor=<rgba 0 0 0 0>}
       cornerRadii=<4 4 4 4>
      }
      tintColor {colorUseRule=UseColor_Specified, specifiedColor=<rgba 0.0284 0.0284 0.0284 1>}
      uVRegion {bIsValid=false, max=<0 0>, min=<0 0>}
     }
     pressedForeground {colorUseRule=UseColor_ColorTable, specifiedColor=<rgba 0 0 0 0>}
     pressedPadding=<ltrb 12 2 12 0>
     pressedSlateSound {resourceObject=null}
    }
   }
   slot {
    class=CanvasPanelSlot
    properties {layoutData {alignment=<0 0>, anchors {maximum=<0 0>, minimum=<0 0>}, offsets=<ltrb 19.454 115.454 404.4863 90.414>}}
   }
  }
  CommonTextBlock {
   widgetClassPath=&CommonTextBlock
   parent=Button_CreateMaterial
   widget=CommonTextBlock
   properties {
    justification=Center
    style=&TextStyle-Base_C
    text=Create Material
    font {
     bForceMonospaced=false
     bMaterialIsStencil=false
     fontMaterial=null
     fontObject=&NotoSans
     letterSpacing=0
     monospacedWidth=1
     size=24
     skewAmount=0
     typefaceFontName=NotoSans-Regular
     outlineSettings {
      bApplyOutlineToDropShadows=false
      bMiteredCorners=false
      bSeparateFillAlpha=false
      outlineMaterial=null
      outlineSize=0
      outlineColor=<rgba 0 0 0 1>
     }
    }
    shadowOffset=<0 0>
   }
   slot {class=ButtonSlot}
  }
 ]
}
