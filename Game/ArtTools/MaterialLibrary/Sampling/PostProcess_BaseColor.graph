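# Post-process material in the art tools sampling library, named for showing the scene's base color.
# Its only node is a SceneTexture set to StoredBaseColor, the base color as stored in the GBuffer, with no output link shown.
# It is set to replace the tonemapper, so it takes the place of that pass when blended in.
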Material /Game/ArtTools/MaterialLibrary/Sampling/PostProcess_BaseColor
graph {
 properties {
  blendableLocation=BL_ReplacingTonemapper
  materialDomain=MD_PostProcess
  pixelDepthOffsetMode=PDOM_Legacy
  shadingModel=MSM_Unlit
  usedShadingModels=Unlit
 }
 nodes [SceneTexture() {sceneTextureId=PPI_StoredBaseColor}]
}
