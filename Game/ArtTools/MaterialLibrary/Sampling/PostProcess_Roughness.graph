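# Post-process material in the ArtTools material library's Sampling folder, built around reading the scene's GBuffer roughness.
# Its graph holds a single SceneTexture node set to Roughness, with no math, parameters or other processing.
# The file shows no link from that node to a material output.
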Material /Game/ArtTools/MaterialLibrary/Sampling/PostProcess_Roughness
graph {
 properties {materialDomain=MD_PostProcess, pixelDepthOffsetMode=PDOM_Legacy, usedShadingModels=Default Lit}
 nodes [SceneTexture() {sceneTextureId=PPI_Roughness}]
}
